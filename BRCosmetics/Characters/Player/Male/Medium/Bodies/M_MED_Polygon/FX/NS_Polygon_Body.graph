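# Niagara body effect for the Polygon player cosmetic, layering rising polygon meshes, falling pyramid debris and flickering lightning sprites.
# Polygon_Bits forms two offset spirals climbing the body, shrinking as camera distance closes.
# GPU pyramids burst from a rising cylinder, fall with delayed collision, and pass User.Level to their material.
# Two GPU lightning emitters sample the User body and face skeletal meshes, sleeping at close range through minimum-distance culling.
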NiagaraSystem /BRCosmetics/Characters/Player/Male/Medium/Bodies/M_MED_Polygon/FX/NS_Polygon_Body
dependencies {
 usedDataInterfaces [&NiagaraDataInterfaceCurve, &NiagaraDataInterfaceColorCurve, &NiagaraDataInterfaceVector2DCurve]
 usedDynamicInputs []
 usedModules [&SystemState]
}
emitters {
 Lightning_Pulse {
  simTarget=GPUComputeSim
  emitterUpdateScript [
   &EmitterState(
    Enable Distance Culling=true
    Life Cycle Mode=::NewEnumerator1
    Loop Behavior=::NewEnumerator1
    Loop Delay=0.2
    Loop Duration=0.2
    Min Distance Response=::SleepAndLetParticlesFinish
    MinDistance=250
    Module.bUseMaxDistance=false
    Module.bUseMinDistance=true
    Scalability Mode=::NewEnumerator1
    UseLoopDelay=true
   )
   &SpawnRate(SpawnGroup=1, SpawnRate=10, Use Spawn Probability=false) {module=SpawnRate001}
  ]
  particleSpawnScript [
   &InitializeParticle(
    AdjustAlpha=false
    AdjustHue=false
    AdjustSaturation=false
    AdjustValue=true
    Color=<rgba 0.9301 0.7454 0.1195 1>
    Color Mode=::NewEnumerator3
    Lifetime Max=0.2
    Lifetime Min=0.1
    Lifetime Mode=::NewEnumerator1
    Sprite Rotation Mode=::NewEnumerator2
    Sprite Size Mode=::NewEnumerator1
    Uniform Sprite Size Max=20
    UsePositionOffset=false
    Value Range:Vector2f=<1 150>
    Write Material Random=true
   )
   &CylinderLocation(
    Cylinder Height:NiagaraFloat=&FloatFromCurve(
     FloatCurve=DataInterface() {propertyValues {Curve {keys [{arriveTangent=328.8044, interpMode=RCIM_Cubic, leaveTangent=328.8044, tangentMode=RCTM_User, time=0, value=0}, <at 1 160 RCIM_Cubic>]}}}
     CurveIndex=$Emitter.NormalizedLoopAge
    )
    Cylinder Height Midpoint=0
    Cylinder Origin:NiagaraPosition=/Niagara/DynamicInputs/Helpers/SimulationPosition()
    Cylinder Radius=50
    MaskSpawnGroup=false
    Override Seed=false
    Surface Only=true
    Use Endcaps In Surface Only Mode=false
    UseOwnerScale=true
   )
   &SkeletalMeshLocation(
    MaskSpawnGroup=true
    Mesh Sampling Type=::NewEnumerator7
    OffsetAlongNormal=false
    OverrideRandomSeed=false
    Skeletal Mesh=$User.SkeletalMesh_Face
    Spawn Group Mask=1
   ) {module=SkeletalMeshLocation001}
   &CameraOffset(Camera Offset Amount=5)
   &SubUVAnimation(End Frame=3, SubUV Animation Mode:ENiagara_SubUVLookupMode=::NewEnumerator2)
  ]
  particleUpdateScript [
   &ParticleState()
   &ScaleColor(
    Scale Alpha=&FloatFromCurve(FloatCurve=DataInterface() {propertyValues {Curve {keys [<at 0.5 1 RCIM_Cubic>, {arriveTangent=-3.193, interpMode=RCIM_Cubic, leaveTangent=-3.193, tangentMode=RCTM_User, time=1, value=0}]}}})
    Scale RGB=&VectorFromFloat(Value=25)
    ScaleA=true
    ScaleRGB=true
   )
   &DynamicMaterialParameters(
    Index 0 Param 2=2
    Index 0 Param 3=1
    Index 0 Param 4=2
    Module.Param0WriteEnabled=true
    Module.Param1WriteEnabled=true
    Module.Param2WriteEnabled=true
    Module.Param3WriteEnabled=true
   )
   &ScaleSpriteSize(
    Non-Uniform Curve Sprite Scale=DataInterface() {propertyValues {
      XCurve {keys [{arriveTangent=0.4736, interpMode=RCIM_Cubic, leaveTangent=0.4736, tangentMode=RCTM_User, time=0, value=0.8}, <at 1 1 RCIM_Cubic>]}
      YCurve {keys [
        {interpMode=RCIM_Cubic, leaveTangent=2.4085, tangentMode=RCTM_Break, time=0, value=0.2}
        {interpMode=RCIM_Cubic, tangentMode=RCTM_Break, time=1, value=1}
       ]}
     }}
    Scale Sprite Size Mode=::NewEnumerator3
   )
  ]
  rendererClasses [&NiagaraSpriteRendererProperties]
  renderers [{rendererClass=&NiagaraSpriteRendererProperties}]
 }
 Lightning_Pulse001 {
  simTarget=GPUComputeSim
  emitterUpdateScript [
   &EmitterState(
    Enable Distance Culling=true
    Life Cycle Mode=::NewEnumerator1
    Loop Behavior=::NewEnumerator1
    Loop Delay=0.2
    Loop Duration=0.2
    Min Distance Response=::SleepAndLetParticlesFinish
    MinDistance=250
    Module.bUseMaxDistance=false
    Module.bUseMinDistance=true
    Scalability Mode=::NewEnumerator1
    UseLoopDelay=true
   )
   &SpawnRate(SpawnRate=50, Use Spawn Probability=false)
  ]
  particleSpawnScript [
   &InitializeParticle(
    AdjustAlpha=false
    AdjustHue=false
    AdjustSaturation=false
    AdjustValue=true
    Color=<rgba 0.9301 0.7454 0.1195 1>
    Color Mode=::NewEnumerator3
    Lifetime Max=0.2
    Lifetime Min=0.1
    Lifetime Mode=::NewEnumerator1
    Sprite Rotation Mode=::NewEnumerator2
    Sprite Size Mode=::NewEnumerator1
    Uniform Sprite Size Max=20
    UsePositionOffset=false
    Value Range:Vector2f=<1 150>
    Write Material Random=true
   )
   &CylinderLocation(
    Cylinder Height:NiagaraFloat=&FloatFromCurve(
     FloatCurve=DataInterface() {propertyValues {Curve {keys [{arriveTangent=328.8044, interpMode=RCIM_Cubic, leaveTangent=328.8044, tangentMode=RCTM_User, time=0, value=0}, <at 1 160 RCIM_Cubic>]}}}
     CurveIndex=$Emitter.NormalizedLoopAge
    )
    Cylinder Height Midpoint=0
    Cylinder Origin:NiagaraPosition=&SimulationPosition()
    Cylinder Radius=50
    MaskSpawnGroup=false
    Override Seed=false
    Surface Only=true
    Use Endcaps In Surface Only Mode=false
    UseOwnerScale=true
   )
   &SkeletalMeshLocation(
    MaskSpawnGroup=true
    Mesh Sampling Type=::NewEnumerator7
    OffsetAlongNormal=false
    OverrideRandomSeed=false
    Skeletal Mesh=$User.SkeletalMesh
   )
   &CameraOffset(Camera Offset Amount=5)
   &SubUVAnimation(End Frame=3, SubUV Animation Mode:ENiagara_SubUVLookupMode=::NewEnumerator2)
  ]
  particleUpdateScript [
   &ParticleState()
   &ScaleColor(
    Scale Alpha=&FloatFromCurve(FloatCurve=DataInterface() {propertyValues {Curve {keys [<at 0.5 1 RCIM_Cubic>, {arriveTangent=-3.193, interpMode=RCIM_Cubic, leaveTangent=-3.193, tangentMode=RCTM_User, time=1, value=0}]}}})
    Scale RGB=&VectorFromFloat(Value=25)
    ScaleA=true
    ScaleRGB=true
   )
   &DynamicMaterialParameters(
    Index 0 Param 2=2
    Index 0 Param 3=1
    Index 0 Param 4=2
    Module.Param0WriteEnabled=true
    Module.Param1WriteEnabled=true
    Module.Param2WriteEnabled=true
    Module.Param3WriteEnabled=true
   )
   &ScaleSpriteSize(
    Non-Uniform Curve Sprite Scale=DataInterface() {propertyValues {
      XCurve {keys [{arriveTangent=0.4736, interpMode=RCIM_Cubic, leaveTangent=0.4736, tangentMode=RCTM_User, time=0, value=0.8}, <at 1 1 RCIM_Cubic>]}
      YCurve {keys [
        {interpMode=RCIM_Cubic, leaveTangent=2.4085, tangentMode=RCTM_Break, time=0, value=0.2}
        {interpMode=RCIM_Cubic, tangentMode=RCTM_Break, time=1, value=1}
       ]}
     }}
    Scale Sprite Size Mode=::NewEnumerator3
   )
  ]
  rendererClasses [&NiagaraSpriteRendererProperties]
  renderers [{rendererClass=&NiagaraSpriteRendererProperties}]
 }
 Polygon_Bits {
  emitterUpdateScript [
   &EmitterState()
   &SpawnRate(SpawnRate=60, Use Spawn Probability=false)
   &SpawnRate(SpawnGroup=1, SpawnRate=60, Use Spawn Probability=false) {module=SpawnRate001}
  ]
  particleSpawnScript [
   &InitializeParticle(
    Color Mode=::NewEnumerator0
    Lifetime=0.5
    Mesh Scale Mode=::NewEnumerator1
    Mesh Uniform Scale Max=1
    Mesh Uniform Scale Min=0.75
    UsePositionOffset=false
    Write Material Random=true
   )
   &TorusLocation(
    Align Torus With Axis=false
    Handle Radius=1
    Large Radius=25
    MaskSpawnGroup=false
    Torus Distribution Mode:ENiagaraTorusDistributionMode=::NewEnumerator1
    Torus Origin:NiagaraPosition=&Add_Vector(
     A=&SimulationPosition()
     B=&MakeVector(X=0, Y=0, Z=&Lerp_Float(A=0, B=170, Alpha=$Emitter.NormalizedLoopAge))
    )
    U Position:NiagaraFloat=&Lerp_Float(
     A=&Lerp_Float(A=0, B=0.5, Alpha=&MakeFloatFromInt(Int=&SpawnGroup()))
     B=&Lerp_Float(A=2, B=2.5, Alpha=&MakeFloatFromInt(Int=&SpawnGroup()))
     Alpha=$Emitter.NormalizedLoopAge
    )
   )
   &InitialMeshOrientation(
    Orientation Axis=<0 0 1>
    Rotation=&MakeVector(
     X=0
     Y=0
     Z=&Lerp_Float(A=0, B=0.5, Alpha=&MakeFloatFromInt(Int=&SpawnGroup()))
    )
    Use Orientation Vector=true
    Use Rotation Vector=true
   )
  ]
  particleUpdateScript [
   &ParticleState()
   &UpdateMeshOrientation(
    Rotation Rate=&FloatFromCurve(
     FloatCurve=DataInterface() {propertyValues {Curve {keys [<at 0.25 0 RCIM_Cubic>, {arriveTangent=4.738, interpMode=RCIM_Cubic, leaveTangent=4.738, tangentMode=RCTM_User, time=0.75, value=1}]}}}
     Scale Curve=3
    )
    Rotation Vector=<0 0 1>
   )
   &ScaleColor(
    Scale RGB=&MakeVectorFromLinearColorRGB(Color=&ColorFromCurve(ColorCurve=DataInterface() {propertyValues {
        AlphaCurve {keys []}
        BlueCurve {keys [
          {interpMode=RCIM_Cubic, leaveTangentWeight=0.1667, tangentMode=RCTM_User, time=0, value=0}
          {arriveTangentWeight=0.1667, interpMode=RCIM_Cubic, leaveTangentWeight=0.1667, tangentMode=RCTM_User, time=0.5, value=0.0908}
          <at 1 0.1195>
         ]}
        GreenCurve {keys [
          {interpMode=RCIM_Cubic, leaveTangentWeight=0.1667, tangentMode=RCTM_User, time=0, value=0}
          {arriveTangentWeight=0.1667, interpMode=RCIM_Cubic, leaveTangentWeight=0.1667, tangentMode=RCTM_User, time=0.5, value=0.0395}
          <at 1 0.7454>
         ]}
        RedCurve {keys [
          {interpMode=RCIM_Cubic, leaveTangentWeight=0.1667, tangentMode=RCTM_User, time=0, value=0}
          {
           arriveTangent=3.4566
           arriveTangentWeight=0.1667
           interpMode=RCIM_Cubic
           leaveTangent=-2.8352
           leaveTangentWeight=0.1667
           tangentMode=RCTM_Break
           time=0.5
           value=1
          }
          <at 1 0.9301>
         ]}
       }}))
    ScaleA=true
    ScaleRGB=true
   )
   &DynamicMaterialParameters(
    Index 0 Param 1=&FloatFromCurve(
     FloatCurve=DataInterface() {propertyValues {Curve {keys [
         {interpMode=RCIM_Cubic, tangentMode=RCTM_Break, time=0, value=0}
         {arriveTangent=7.0438, interpMode=RCIM_Cubic, leaveTangent=-6.0805, tangentMode=RCTM_Break, time=0.5, value=1}
         {interpMode=RCIM_Cubic, tangentMode=RCTM_Break, time=1, value=0}
        ]}}}
     Scale Curve=10
    )
    Module.Param0WriteEnabled=true
    Module.Param1WriteEnabled=false
    Module.Param2WriteEnabled=false
    Module.Param3WriteEnabled=false
   )
   &ScaleMeshSize(Scale Factor=&VectorFromFloat(Value=&FloatFromCurve(FloatCurve=DataInterface() {propertyValues {Curve {keys [
          {arriveTangent=3.5257, interpMode=RCIM_Cubic, leaveTangent=3.5257, leaveTangentWeight=0.1667, tangentMode=RCTM_User, time=0, value=0}
          {arriveTangentWeight=0.7079, interpMode=RCIM_Cubic, leaveTangentWeight=0.8323, tangentMode=RCTM_Break, time=0.5, value=1.25}
          {arriveTangent=-3.6464, arriveTangentWeight=0.7828, interpMode=RCIM_Cubic, leaveTangent=-3.6464, tangentMode=RCTM_User, time=1, value=0}
         ]}}})))
   &ScaleMeshSize(Scale Factor=&VectorFromFloat(Value=&RemapRange(Input Value To Remap=$Engine.Owner.LODDistance, Input Min=0, Input Max=250, Output Min=0, Output Max=1))) {module=ScaleMeshSize001}
   &AccelerationForce(Acceleration=<0 0 150>)
   &CurlNoiseForce(
    Noise Frequency=20
    Noise Strength=&FloatFromCurve(
     FloatCurve=DataInterface() {propertyValues {Curve {keys [
         {interpMode=RCIM_Cubic, leaveTangentWeight=0.2922, tangentMode=RCTM_User, time=0.25, value=0}
         {arriveTangent=2.7275, interpMode=RCIM_Cubic, leaveTangent=2.7275, tangentMode=RCTM_User, time=1, value=1}
        ]}}}
     Scale Curve=2500
    )
    Pan Noise=false
   )
   &InheritSourceMovement(
    Applied Position Scale=&VectorFromFloat(Value=&FloatFromCurve(FloatCurve=DataInterface() {propertyValues {Curve {keys [{arriveTangent=-6.7515, interpMode=RCIM_Cubic, leaveTangent=-6.7515, tangentMode=RCTM_User, time=0.5, value=1}, <at 1 0 RCIM_Cubic>]}}}))
    ApplyForce=false
    ApplyPosition=true
    ApplyPositionFalloff=false
   )
   &SolveForcesAndVelocity(Clamp Velocity=false, Limit Acceleration=false)
  ]
  rendererClasses [&NiagaraMeshRendererProperties]
  renderers [{rendererClass=&NiagaraMeshRendererProperties, rendererIndex=0}]
 }
 Pyramid_Drop_Offs {
  simTarget=GPUComputeSim
  emitterUpdateScript [
   &EmitterState(Life Cycle Mode=::NewEnumerator1, Loop Behavior=::NewEnumerator1, Loop Delay=0.1, Loop Duration=0.4, UseLoopDelay=true)
   &SpawnRate(SpawnRate=350, Use Spawn Probability=false)
  ]
  particleSpawnScript [
   &InitializeParticle(
    Color=<rgba 1 0.0395 0.0908 1>
    Lifetime Max=1.25
    Lifetime Min=0.75
    Lifetime Mode=::NewEnumerator1
    Mesh Scale Mode=::NewEnumerator1
    Mesh Uniform Scale Max=0.05
    Mesh Uniform Scale Min=0.025
    UsePositionOffset=false
    Write Material Random=true
   )
   &CylinderLocation(
    Cylinder Height:NiagaraFloat=&FloatFromCurve(
     FloatCurve=DataInterface() {propertyValues {Curve {keys [{arriveTangent=328.8044, interpMode=RCIM_Cubic, leaveTangent=328.8044, tangentMode=RCTM_User, time=0, value=0}, <at 1 160 RCIM_Cubic>]}}}
     CurveIndex=$Emitter.NormalizedLoopAge
    )
    Cylinder Height Midpoint=0
    Cylinder Origin:NiagaraPosition=&SimulationPosition()
    Cylinder Radius=30
    MaskSpawnGroup=false
    Override Seed=false
    Surface Only=true
    Use Endcaps In Surface Only Mode=false
    UseOwnerScale=true
   )
   /SetVariables(Particles.IntPosition:NiagaraPosition=$Particles.Position) {isSetParameters=true}
   &InitialMeshOrientation(
    Rotation=&RandomRangeVector(Minimum=<0 0 0>, Maximum=<1 1 1>)
    Use Orientation Vector=true
    Use Rotation Vector=true
   )
   &CameraOffset(Camera Offset Amount=5)
   &AddVelocityFromPoint(
    Module.Use Falloff Distance=false
    Velocity Origin:Vector3f=&Add_Vector(
     A=&SimulationPosition()
     B=&MakeVector(
      X=0
      Y=0
      Z=&FloatFromCurve(
       FloatCurve=DataInterface() {propertyValues {Curve {keys [{arriveTangent=328.8044, interpMode=RCIM_Cubic, leaveTangent=328.8044, tangentMode=RCTM_User, time=0, value=0}, <at 1 160 RCIM_Cubic>]}}}
       CurveIndex=$Emitter.NormalizedLoopAge
      )
     )
    )
    Velocity Strength=&Multiply_Float(
     A=100
     B=&FloatFromCurve(FloatCurve=DataInterface() {propertyValues {Curve {keys [<at 0 1>,<at 1 2>]}}}, CurveIndex=$Emitter.NormalizedLoopAge)
    )
   )
  ]
  particleUpdateScript [
   &ParticleState()
   &DynamicMaterialParameters(
    Index 0 Param 1=&MakeFloatFromInt(Int=$User.Level)
    Index 0 Param 2=&FloatFromCurve(
     FloatCurve=DataInterface() {propertyValues {Curve {keys [
         {arriveTangent=58.5643, interpMode=RCIM_Cubic, leaveTangent=58.5643, tangentMode=RCTM_User, time=0, value=0}
         {arriveTangent=-49.7173, interpMode=RCIM_Cubic, leaveTangent=-49.7173, tangentMode=RCTM_User, time=0.15, value=5}
         {interpMode=RCIM_Cubic, leaveTangent=-0.0405, time=0.5, value=0}
        ]}}}
     Scale Curve=5
    )
    Module.Param0WriteEnabled=true
    Module.Param1WriteEnabled=true
    Module.Param2WriteEnabled=true
    Module.Param3WriteEnabled=true
   )
   &UpdateMeshOrientation(
    Rotation Rate=&FloatFromCurve(FloatCurve=DataInterface() {propertyValues {Curve {keys [
         {arriveTangent=4.2642, interpMode=RCIM_Cubic, leaveTangent=4.2642, tangentMode=RCTM_User, time=0.25, value=0}
         {interpMode=RCIM_Cubic, leaveTangent=2.2442, time=0.75, value=1}
        ]}}})
    Rotation Vector=<0 0 1>
   )
   &ScaleColor(
    Scale Alpha=&FloatFromCurve(FloatCurve=DataInterface() {propertyValues {Curve {keys [<at 0.5 1 RCIM_Cubic>, {arriveTangent=-3.193, interpMode=RCIM_Cubic, leaveTangent=-3.193, tangentMode=RCTM_User, time=1, value=0}]}}})
    ScaleA=true
    ScaleRGB=true
   )
   &ScaleMeshSize(Scale Factor=&VectorFromFloat(Value=&FloatFromCurve(FloatCurve=DataInterface() {propertyValues {Curve {keys [
          {arriveTangent=3.5257, interpMode=RCIM_Cubic, leaveTangent=3.5257, leaveTangentWeight=0.1667, tangentMode=RCTM_User, time=0, value=0}
          {arriveTangentWeight=0.7079, interpMode=RCIM_Cubic, leaveTangentWeight=0.8323, tangentMode=RCTM_Break, time=0.5, value=1}
          {arriveTangent=-3.6464, arriveTangentWeight=0.7828, interpMode=RCIM_Cubic, leaveTangent=-3.6464, tangentMode=RCTM_User, time=1, value=0}
         ]}}})))
   &CurlNoiseForce(Noise Frequency=20, Noise Strength=200, Pan Noise=false)
   &GravityForce(Gravity=&Multiply_VectorByFloat(
     Vector=<0 0 -980>
     Float=&FloatFromCurve(FloatCurve=DataInterface() {propertyValues {Curve {keys [
          {arriveTangent=2.8823, interpMode=RCIM_Cubic, leaveTangent=2.8823, tangentMode=RCTM_User, time=0, value=0}
          {arriveTangentWeight=0.7704, interpMode=RCIM_Cubic, time=1, value=1}
         ]}}})
    ))
   &InheritSourceMovement(
    Applied Position Scale=&VectorFromFloat(Value=&FloatFromCurve(FloatCurve=DataInterface() {propertyValues {Curve {keys [
          {arriveTangent=-2.1557, interpMode=RCIM_Cubic, leaveTangent=-2.1557, tangentMode=RCTM_User, time=0.25, value=1}
          {arriveTangentWeight=0.25, interpMode=RCIM_Cubic, tangentMode=RCTM_User, time=1, value=0}
         ]}}}))
    ApplyForce=false
    ApplyPosition=true
    ApplyPositionFalloff=false
   )
   &Drag(Drag=&RandomRangeFloat(Minimum=2, Maximum=4), Use Linear Drag=true, Use Rotational Drag=false)
   &Collision(
    Age Colliding Particles=false
    Collision Enabled=&SetBoolByFloatComparison(A=$Particles.NormalizedAge, B=0.333, Comparison Type=::NewEnumerator2)
    Control Roll On Collision=true
    GPU Collision Type=::NewEnumerator2
    Mesh Dimensions:Vector3f=<181 181 90>
    Module.Control Rest State Via Surface Normal=false
    Module.EnableMaxCollisionCount=false
    Module.Randomize Collision Normal Vector=false
    Radius Calculation Type=::NewEnumerator1
    ScaleRateOfRotation=false
   )
   &SolveForcesAndVelocity(Clamp Velocity=false, Limit Acceleration=false)
  ]
  rendererClasses [&NiagaraMeshRendererProperties]
  renderers [{rendererClass=&NiagaraMeshRendererProperties, rendererIndex=0}]
 }
}
summary {
 systemName=NS_Polygon_Body
 emitters [
  {emitterName=Polygon_Bits, rendererClasses [&NiagaraMeshRendererProperties]}
  {emitterName=Pyramid_Drop_Offs, simTarget=GPUComputeSim, rendererClasses [&NiagaraMeshRendererProperties]}
  {emitterName=Lightning_Pulse, simTarget=GPUComputeSim, rendererClasses [&NiagaraSpriteRendererProperties]}
  {emitterName=Lightning_Pulse001, simTarget=GPUComputeSim, rendererClasses [&NiagaraSpriteRendererProperties]}
 ]
 userVariables [
  User.Level {type=NiagaraInt32}
  User.SkeletalMesh {
   type=NiagaraDataInterfaceSkeletalMesh
   defaultValue {
    struct=NiagaraExt_VariableValue_DataInterface
    value {dataInterface=&self:NiagaraDataInterfaceSkeletalMesh_0}
   }
  }
  User.SkeletalMesh_Face {
   type=NiagaraDataInterfaceSkeletalMesh
   defaultValue {
    struct=NiagaraExt_VariableValue_DataInterface
    value {dataInterface=&self:NiagaraDataInterfaceSkeletalMesh_3}
   }
  }
 ]
}
