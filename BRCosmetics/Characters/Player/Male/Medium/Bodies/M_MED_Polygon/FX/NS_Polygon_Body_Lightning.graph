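# Niagara cosmetic effect that flickers short-lived orange lightning sprites over the Polygon character's body and head.
# Two GPU emitters spawn flipbook sprites on skeletal mesh surfaces: random body triangles and a separate face mesh.
# User flags drive visibility tags, using triangle ID for hidden hands and feet; a hidden head also stops head spawning.
# Curves animate size, alpha and time-of-day-scaled brightness, and both emitters sleep under minimum-distance culling.
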NiagaraSystem /BRCosmetics/Characters/Player/Male/Medium/Bodies/M_MED_Polygon/FX/NS_Polygon_Body_Lightning
dependencies {
 usedDataInterfaces [&NiagaraDataInterfaceCurve, &NiagaraDataInterfaceVector2DCurve]
 usedDynamicInputs []
 usedModules [
  &self:SystemSpawnScript.SystemScriptSource.SystemScriptGraph.NiagaraNodeAssignment_0.SetVariables_#1
  &SystemState
 ]
}
emitters {
 Lightning_Pulse {
  simTarget=GPUComputeSim
  emitterUpdateScript [
   &EmitterState(
    Enable Distance Culling=true
    Life Cycle Mode=::NewEnumerator1
    Loop Delay=0.2
    Loop Duration=0.2
    Min Distance Response=::SleepAndLetParticlesFinish
    MinDistance=250
    Module.bUseMaxDistance=false
    Module.bUseMinDistance=true
    Scalability Mode=::NewEnumerator1
    UseLoopDelay=false
   )
   &SpawnRate(SpawnRate=&RandomRangeFloat(Minimum=0, Maximum=50), Use Spawn Probability=false)
  ]
  particleSpawnScript [
   /SetVariables(Particles.RandomTri:NiagaraInt32=&RandomRangeInt(Minimum=0, Maximum=225)) {isSetParameters=true}
   &InitializeParticle(
    Color Maximum=<rgba 3 1.6 0 1>
    Color Minimum=<rgba 3 1.6 0 1>
    Color Mode=::NewEnumerator2
    Lifetime Max=0.25
    Lifetime Min=0.13
    Lifetime Mode=::NewEnumerator1
    Position Offset=/Niagara/DynamicInputs/Helpers/SimulationPosition()
    Sprite Rotation Mode=::NewEnumerator2
    Sprite Size Mode=::NewEnumerator1
    Uniform Sprite Size Max=20
    UsePositionOffset=true
    Write Material Random=true
   )
   &SkeletalMeshLocation(
    FixedBarycentric=false
    MaskSpawnGroup=true
    Mesh Sampling Type=::NewEnumerator7
    OffsetAlongNormal=false
    Skeletal Mesh=$System.Skeletal Mesh
    Triangle ID:NiagaraInt32=$Particles.RandomTri
    Triangle Sampling Mode:ENiagaraTriangleSamplingMode=::NewEnumerator7
   )
   &CameraOffset(Camera Offset Amount=5)
   &SubUVAnimation(End Frame=3, SubUV Animation Mode:ENiagara_SubUVLookupMode=::NewEnumerator2)
  ]
  particleUpdateScript [
   &ParticleState()
   &ScaleColor(
    Scale Alpha=&FloatFromCurve(FloatCurve=DataInterface() {propertyValues {Curve {keys [<at 0.5 1 RCIM_Cubic>, {arriveTangent=-3.193, interpMode=RCIM_Cubic, leaveTangent=-3.193, tangentMode=RCTM_User, time=1, value=0}]}}})
    Scale RGB=&Multiply_VectorByFloat(
     Vector=<2 2 2>
     Float=&FloatFromCurve(
      FloatCurve=DataInterface() {propertyValues {Curve {keys [
          {arriveTangent=3.3951, interpMode=RCIM_Cubic, leaveTangent=3.3951, tangentMode=RCTM_User, time=0, value=0.5}
          {interpMode=RCIM_Cubic, tangentMode=RCTM_User, time=0.333, value=1}
          {interpMode=RCIM_Cubic, tangentMode=RCTM_User, time=1, value=0.5}
         ]}}}
      Scale Curve=&FNTimeOfDayScalar(Night Scale=0.8, Morning Scale=1, Evening Scale=0.8)
     )
    )
    ScaleA=true
    ScaleRGB=true
   )
   &DynamicMaterialParameters(
    Index 0 Param 2=2
    Index 0 Param 4=2
    Module.Param0WriteEnabled=true
    Module.Param1WriteEnabled=true
    Module.Param2WriteEnabled=true
    Module.Param3WriteEnabled=true
   )
   &ScaleSpriteSize(
    Non-Uniform Curve Sprite Scale=DataInterface() {propertyValues {
      XCurve {keys [{arriveTangent=0.4736, interpMode=RCIM_Cubic, leaveTangent=0.4736, tangentMode=RCTM_User, time=0, value=0.8}, <at 1 1 RCIM_Cubic>]}
      YCurve {keys [
        {interpMode=RCIM_Cubic, leaveTangent=2.4085, tangentMode=RCTM_Break, time=0, value=0.2}
        {interpMode=RCIM_Cubic, tangentMode=RCTM_Break, time=1, value=1}
       ]}
     }}
    Scale Sprite Size Mode=::NewEnumerator3
   )
   /SetVariables(Particles.VisibilityTag=&self:Arms_And_Legs_Hidden_Check(
     is Hand Hidden:NiagaraBool=$User.bisHandHidden
     is Foot Hidden:NiagaraBool=$User.bisFootHidden
     Triangle ID:NiagaraInt32=$Particles.RandomTri
    )) {isSetParameters=true}
  ]
  rendererClasses [&NiagaraSpriteRendererProperties]
  renderers [{rendererClass=&NiagaraSpriteRendererProperties}]
 }
 Lightning_Pulse_Head {
  simTarget=GPUComputeSim
  emitterUpdateScript [
   &EmitterState(
    Enable Distance Culling=true
    Life Cycle Mode=::NewEnumerator1
    Loop Delay=0.2
    Loop Duration=0.2
    Min Distance Response=::SleepAndLetParticlesFinish
    MinDistance=250
    Module.bUseMaxDistance=false
    Module.bUseMinDistance=true
    Scalability Mode=::NewEnumerator1
    UseLoopDelay=false
   )
   &SpawnRate(
    SpawnRate=&MakeCustomFloatFromBool(Bool=$User.bisHeadHidden, True Float=0, False Float=&RandomRangeFloat(Minimum=0, Maximum=2))
    Use Spawn Probability=false
   ) {module=SpawnRate001}
  ]
  particleSpawnScript [
   &InitializeParticle(
    Color Maximum=<rgba 3 1.6 0 1>
    Color Minimum=<rgba 3 1.6 0 1>
    Color Mode=::NewEnumerator2
    Lifetime Max=0.25
    Lifetime Min=0.13
    Lifetime Mode=::NewEnumerator1
    Position Offset=&SimulationPosition()
    Sprite Rotation Mode=::NewEnumerator2
    Sprite Size Mode=::NewEnumerator1
    Uniform Sprite Size Max=20
    UsePositionOffset=true
    Write Material Random=true
   )
   &SkeletalMeshLocation(
    MaskSpawnGroup=true
    Mesh Sampling Type=::NewEnumerator7
    OffsetAlongNormal=false
    OverrideRandomSeed=false
    Skeletal Mesh=$User.SkeletalMesh_Face
   ) {module=SkeletalMeshLocation001}
   &CameraOffset(Camera Offset Amount=5)
   &SubUVAnimation(End Frame=3, SubUV Animation Mode:ENiagara_SubUVLookupMode=::NewEnumerator2)
  ]
  particleUpdateScript [
   &ParticleState()
   &ScaleColor(
    Scale Alpha=&FloatFromCurve(FloatCurve=DataInterface() {propertyValues {Curve {keys [<at 0.5 1 RCIM_Cubic>, {arriveTangent=-3.193, interpMode=RCIM_Cubic, leaveTangent=-3.193, tangentMode=RCTM_User, time=1, value=0}]}}})
    Scale RGB=&Multiply_VectorByFloat(
     Vector=<2 2 2>
     Float=&FloatFromCurve(
      FloatCurve=DataInterface() {propertyValues {Curve {keys [
          {arriveTangent=3.3951, interpMode=RCIM_Cubic, leaveTangent=3.3951, tangentMode=RCTM_User, time=0, value=0.5}
          {interpMode=RCIM_Cubic, tangentMode=RCTM_User, time=0.333, value=1}
          {interpMode=RCIM_Cubic, tangentMode=RCTM_User, time=1, value=0.5}
         ]}}}
      Scale Curve=&FNTimeOfDayScalar(Night Scale=0.8, Morning Scale=1, Evening Scale=0.8)
     )
    )
    ScaleA=true
    ScaleRGB=true
   )
   &DynamicMaterialParameters(
    Index 0 Param 2=2
    Index 0 Param 4=2
    Module.Param0WriteEnabled=true
    Module.Param1WriteEnabled=true
    Module.Param2WriteEnabled=true
    Module.Param3WriteEnabled=true
   )
   &ScaleSpriteSize(
    Non-Uniform Curve Sprite Scale=DataInterface() {propertyValues {
      XCurve {keys [{arriveTangent=0.4736, interpMode=RCIM_Cubic, leaveTangent=0.4736, tangentMode=RCTM_User, time=0, value=0.8}, <at 1 1 RCIM_Cubic>]}
      YCurve {keys [
        {interpMode=RCIM_Cubic, leaveTangent=2.4085, tangentMode=RCTM_Break, time=0, value=0.2}
        {interpMode=RCIM_Cubic, tangentMode=RCTM_Break, time=1, value=1}
       ]}
     }}
    Scale Sprite Size Mode=::NewEnumerator3
   )
   /SetVariables(Particles.VisibilityTag=&MakeCustomIntFromBool(Bool=$User.bisHeadHidden, True INT=1, False INT=0)) {isSetParameters=true}
  ]
  rendererClasses [&NiagaraSpriteRendererProperties]
  renderers [{rendererClass=&NiagaraSpriteRendererProperties}]
 }
}
summary {
 systemName=NS_Polygon_Body_Lightning
 emitters [
  {emitterName=Lightning_Pulse, simTarget=GPUComputeSim, rendererClasses [&NiagaraSpriteRendererProperties]}
  {emitterName=Lightning_Pulse_Head, simTarget=GPUComputeSim, rendererClasses [&NiagaraSpriteRendererProperties]}
 ]
 userVariables [
  User.bisFootHidden {type=NiagaraBool}
  User.bisHandHidden {type=NiagaraBool}
  User.bisHeadHidden {type=NiagaraBool}
  User.SkeletalMesh_Face {
   type=NiagaraDataInterfaceSkeletalMesh
   defaultValue {
    struct=NiagaraExt_VariableValue_DataInterface
    value {dataInterface=&self:NiagaraDataInterfaceSkeletalMesh_2}
   }
  }
 ]
}
