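# Physics asset for the coat on the medium male Polygon player body, covering dynamic coat bones and the legs.
# Six box-body chains (front, side, back, left and right) each hang from a small sphere base body, with longer chains at the front.
# Tight swing and twist limits link each link to its parent, varying by panel.
# Thigh and calf capsules have no constraints, apparently serving as colliders for the coat.
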PhysicsAsset /BRCosmetics/Characters/Player/Male/Medium/Bodies/M_MED_Polygon/Meshes/M_MED_Polygon_Coat_Physics
bodies {
 calf_l [{shapeType=Capsule, length=28.9879, radius=6.4627, shapeName=calf_l_capsule, center=<-16.6817 -2.0341 0>, rotation=<rot 90 0 0>}]
 calf_r [{shapeType=Capsule, length=27.3503, radius=6.3896, shapeName=calf_r_capsule, center=<16.7122 2.3842 0.0001>, rotation=<rot 90 0 0>}]
 dyn_coat_back_l_01 [{shapeType=Box, extentX=11.3277, extentY=16.488, extentZ=3.3765, shapeName=dyn_coat_back_l_01_box, center=<5.1375 0 -0.0001>}]
 dyn_coat_back_l_base [{shapeType=Sphere, radius=0.5, shapeName=dyn_coat_back_l_base_sphere}]
 dyn_coat_back_l_parent [{shapeType=Box, extentX=7.3793, extentY=15.7805, extentZ=3.3839, shapeName=dyn_coat_back_l_parent_box, center=<3.5234 0 -0.0001>}]
 dyn_coat_back_r_01 [{shapeType=Box, extentX=13.4727, extentY=13.9572, extentZ=3.2934, shapeName=dyn_coat_back_r_01_box, center=<5.9927 0 0>}]
 dyn_coat_back_r_base [{shapeType=Sphere, radius=0.5, shapeName=dyn_coat_back_r_base_sphere}]
 dyn_coat_back_r_parent [{shapeType=Box, extentX=7.6164, extentY=14.0823, extentZ=3.3581, shapeName=dyn_coat_back_r_parent_box, center=<3.7622 0 -0.0001>}]
 dyn_coat_front_l_01 [{shapeType=Box, extentX=12.5791, extentY=15.0596, extentZ=2.0757, shapeName=dyn_coat_front_l_01_box, center=<5.6061 0 0.0001>}]
 dyn_coat_front_l_02 [{shapeType=Box, extentX=7.3014, extentY=7.4943, extentZ=2.6576, shapeName=dyn_coat_front_l_02_box, center=<3.2976 0 0.0001>}]
 dyn_coat_front_l_03 [{shapeType=Box, extentX=6.5215, extentY=4.2872, extentZ=1.7675, shapeName=dyn_coat_front_l_03_box, center=<3.0409 0 0>}]
 dyn_coat_front_l_04 [{shapeType=Box, extentX=3.0427, extentY=4.1188, extentZ=1.8761, shapeName=dyn_coat_front_l_04_box, center=<1.2015 0 -0.0001>}]
 dyn_coat_front_l_05 [{shapeType=Box, extentX=4.4183, extentY=5.8486, extentZ=1.6842, shapeName=dyn_coat_front_l_05_box, center=<1.9659 0 0>}]
 dyn_coat_front_l_base [{shapeType=Sphere, radius=0.5, shapeName=dyn_coat_front_l_base_sphere}]
 dyn_coat_front_l_parent [{shapeType=Box, extentX=10.572, extentY=15.0672, extentZ=2.3625, shapeName=dyn_coat_front_l_parent_box, center=<4.7467 0 0>}]
 dyn_coat_front_r_01 [{shapeType=Box, extentX=12.1762, extentY=20.1413, extentZ=3.9203, shapeName=dyn_coat_front_r_01_box, center=<6.4537 -0.0001 0>}]
 dyn_coat_front_r_02 [{shapeType=Box, extentX=7.8622, extentY=6.2687, extentZ=2.6157, shapeName=dyn_coat_front_r_02_box, center=<3.6422 0 0.0001>}]
 dyn_coat_front_r_03 [{shapeType=Box, extentX=7.2703, extentY=3.8745, extentZ=2.1507, shapeName=dyn_coat_front_r_03_box, center=<3.3831 0 0>}]
 dyn_coat_front_r_04 [{shapeType=Box, extentX=3.4753, extentY=3.9506, extentZ=2.0926, shapeName=dyn_coat_front_r_04_box, center=<1.4321 0 0>}]
 dyn_coat_front_r_05 [{shapeType=Box, extentX=4.502, extentY=6.2662, extentZ=1.4898, shapeName=dyn_coat_front_r_05_box, center=<1.9138 0 -0.0001>}]
 dyn_coat_front_r_base [{shapeType=Sphere, radius=0.5, shapeName=dyn_coat_front_r_base_sphere}]
 dyn_coat_front_r_parent [{shapeType=Box, extentX=10.5259, extentY=20.8116, extentZ=3.2306, shapeName=dyn_coat_front_r_parent_box, center=<5.4522 0 0.0001>}]
 dyn_coat_side_l_01 [{shapeType=Box, extentX=14.2803, extentY=17.7401, extentZ=3.8265, shapeName=dyn_coat_side_l_01_box, center=<6.7319 0 0>}]
 dyn_coat_side_l_base [{shapeType=Sphere, radius=0.5, shapeName=dyn_coat_side_l_base_sphere}]
 dyn_coat_side_l_parent [{shapeType=Box, extentX=11.428, extentY=16.3553, extentZ=5.1847, shapeName=dyn_coat_side_l_parent_box, center=<4.5989 0 0.0001>}]
 dyn_coat_side_r_01 [{shapeType=Box, extentX=13.8167, extentY=15.9523, extentZ=3.335, shapeName=dyn_coat_side_r_01_box, center=<6.9085 0 0>}]
 dyn_coat_side_r_base [{shapeType=Sphere, radius=0.5, shapeName=dyn_coat_side_r_base_sphere}]
 dyn_coat_side_r_parent [{shapeType=Box, extentX=9.4335, extentY=16.068, extentZ=4.0915, shapeName=dyn_coat_side_r_parent_box, center=<4.6946 -0.0001 0>}]
 thigh_l [{shapeType=Capsule, length=33.4228, radius=7.7438, shapeName=thigh_l_capsule, center=<-23.1958 0 0>, rotation=<rot 90 0 0>}]
 thigh_r [{shapeType=Capsule, length=30.5364, radius=7.6061, shapeName=thigh_r_capsule, center=<21.2781 -0.0004 -0.0005>, rotation=<rot 90 0 0>}]
}
constraints [
 {bone1Name=dyn_coat_back_r_01, bone2Name=dyn_coat_back_r_parent, swing1LimitDegrees=10, swing2LimitDegrees=20, twistLimitDegrees=3}
 {bone1Name=dyn_coat_back_r_parent, bone2Name=dyn_coat_back_r_base, swing1LimitDegrees=10, swing2LimitDegrees=20, twistLimitDegrees=3}
 {bone1Name=dyn_coat_back_l_01, bone2Name=dyn_coat_back_l_parent, swing1LimitDegrees=10, swing2LimitDegrees=20, twistLimitDegrees=3}
 {bone1Name=dyn_coat_back_l_parent, bone2Name=dyn_coat_back_l_base, swing1LimitDegrees=10, swing2LimitDegrees=20, twistLimitDegrees=3}
 {bone1Name=dyn_coat_side_r_01, bone2Name=dyn_coat_side_r_parent, swing1LimitDegrees=20, swing2LimitDegrees=20, twistLimitDegrees=3}
 {bone1Name=dyn_coat_side_r_parent, bone2Name=dyn_coat_side_r_base, swing1LimitDegrees=20, swing2LimitDegrees=20, twistLimitDegrees=3}
 {bone1Name=dyn_coat_side_l_01, bone2Name=dyn_coat_side_l_parent, swing1LimitDegrees=20, swing2LimitDegrees=20, twistLimitDegrees=3}
 {bone1Name=dyn_coat_side_l_parent, bone2Name=dyn_coat_side_l_base, swing1LimitDegrees=20, swing2LimitDegrees=20, twistLimitDegrees=3}
 {bone1Name=dyn_coat_front_r_parent, bone2Name=dyn_coat_front_r_base, swing1LimitDegrees=10, swing2LimitDegrees=20, twistLimitDegrees=5}
 {bone1Name=dyn_coat_front_r_01, bone2Name=dyn_coat_front_r_parent, swing1LimitDegrees=10, swing2LimitDegrees=30, twistLimitDegrees=5}
 {bone1Name=dyn_coat_front_r_02, bone2Name=dyn_coat_front_r_01, swing1LimitDegrees=10, swing2LimitDegrees=40, twistLimitDegrees=5}
 {bone1Name=dyn_coat_front_r_03, bone2Name=dyn_coat_front_r_02, swing1LimitDegrees=10, swing2LimitDegrees=40, twistLimitDegrees=5}
 {bone1Name=dyn_coat_front_r_04, bone2Name=dyn_coat_front_r_03, swing1LimitDegrees=10, swing2LimitDegrees=40, twistLimitDegrees=5}
 {bone1Name=dyn_coat_front_r_05, bone2Name=dyn_coat_front_r_04, swing1LimitDegrees=10, swing2LimitDegrees=20, twistLimitDegrees=5}
 {bone1Name=dyn_coat_front_l_parent, bone2Name=dyn_coat_front_l_base, swing1LimitDegrees=10, swing2LimitDegrees=20, twistLimitDegrees=5}
 {bone1Name=dyn_coat_front_l_01, bone2Name=dyn_coat_front_l_parent, swing1LimitDegrees=10, swing2LimitDegrees=30, twistLimitDegrees=5}
 {bone1Name=dyn_coat_front_l_02, bone2Name=dyn_coat_front_l_01, swing1LimitDegrees=10, swing2LimitDegrees=20, twistLimitDegrees=5}
 {bone1Name=dyn_coat_front_l_03, bone2Name=dyn_coat_front_l_02, swing1LimitDegrees=10, swing2LimitDegrees=20, twistLimitDegrees=5}
 {bone1Name=dyn_coat_front_l_04, bone2Name=dyn_coat_front_l_03, swing1LimitDegrees=10, swing2LimitDegrees=20, twistLimitDegrees=5}
 {bone1Name=dyn_coat_front_l_05, bone2Name=dyn_coat_front_l_04, swing1LimitDegrees=5, swing2LimitDegrees=20, twistLimitDegrees=1}
]
