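# Physics asset for the M_MED_Polygon player body, covering the dynamic chest strap and back poncho bones.
# The strap is a chain of four boxes and the poncho a chain of two, each constrained back to a small base sphere.
# Strap joints have tight swing limits, while poncho joints swing much wider, both with little twist.
# Capsules on spine_03 and spine_05 give the torso volume and have no constraints to either chain.
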PhysicsAsset /BRCosmetics/Characters/Player/Male/Medium/Bodies/M_MED_Polygon/Meshes/M_MED_Polygon_Physics
bodies {
 dyn_chest_strap_01 [{shapeType=Box, extentX=3.1703, extentY=2.3276, extentZ=1.491, shapeName=dyn_chest_strap_01_box, center=<1.5747 0 0>}]
 dyn_chest_strap_02 [{shapeType=Box, extentX=2.1078, extentY=2.4718, extentZ=1.4593, shapeName=dyn_chest_strap_02_box, center=<0.9452 0 0>}]
 dyn_chest_strap_03 [{shapeType=Box, extentX=3.8912, extentY=4.9972, extentZ=1.4269, shapeName=dyn_chest_strap_03_box, center=<1.6233 0 0>}]
 dyn_chest_strap_base [{shapeType=Sphere, radius=0.5, shapeName=dyn_chest_strap_base_sphere}]
 dyn_chest_strap_parent [{shapeType=Box, extentX=3.5705, extentY=2.3276, extentZ=1.4899, shapeName=dyn_chest_strap_parent_box, center=<1.8696 0 0>}]
 dyn_poncho_back_01 [{shapeType=Box, extentX=6.8446, extentY=10.0108, extentZ=2.6674, shapeName=dyn_poncho_back_01_box, center=<3.1903 0 0>}]
 dyn_poncho_back_base [{shapeType=Sphere, radius=0.5, shapeName=dyn_poncho_back_base_sphere}]
 dyn_poncho_back_parent [{shapeType=Box, extentX=6.086, extentY=12.8239, extentZ=2.7316, shapeName=dyn_poncho_back_parent_box, center=<2.613 0 0.0001>}]
 spine_03 [
  {shapeType=Capsule, length=13.4624, radius=11.3253, shapeName=spine_03_capsule, center=<-3.0899 -0.7473 0.0002>, rotation=<rot 90 0 -15.6602>}
  {shapeType=Capsule, length=6.2021, radius=9.5364, shapeName=spine_03_capsule, center=<0.0166 3.9637 0>, rotation=<rot -90 -90 95.4132>}
 ]
 spine_05 [{shapeType=Capsule, length=13.2213, radius=11.5014, shapeName=spine_05_capsule, center=<-7.4011 1.4792 0.0004>, rotation=<rot 90 90 84.9449>}]
}
constraints [
 {bone1Name=dyn_poncho_back_01, bone2Name=dyn_poncho_back_parent, swing1LimitDegrees=15, swing2LimitDegrees=60, twistLimitDegrees=5}
 {bone1Name=dyn_poncho_back_parent, bone2Name=dyn_poncho_back_base, swing1LimitDegrees=15, swing2LimitDegrees=60, twistLimitDegrees=5}
 {bone1Name=dyn_chest_strap_parent, bone2Name=dyn_chest_strap_base, swing1LimitDegrees=2, swing2LimitDegrees=3, twistLimitDegrees=3}
 {bone1Name=dyn_chest_strap_01, bone2Name=dyn_chest_strap_parent, swing1LimitDegrees=3, swing2LimitDegrees=5, twistLimitDegrees=3}
 {bone1Name=dyn_chest_strap_02, bone2Name=dyn_chest_strap_01, swing1LimitDegrees=4, swing2LimitDegrees=8, twistLimitDegrees=3}
 {bone1Name=dyn_chest_strap_03, bone2Name=dyn_chest_strap_02, swing1LimitDegrees=2, swing2LimitDegrees=10, twistLimitDegrees=3}
]
