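# Post-copy secondary-motion AnimBP for the M_MED_Polygon player body, animating coat, pouches and chest strap over the attached parent's pose.
# Two RigidBody simulations (body, coat) feed weighted bone constraints, additive and replace bone transforms, and two deformation Control Rigs.
# Thigh swing angles drive corrective alphas; jump and fall events set an external force on the body simulation.
# Crouch_Alpha selects 1 either way, keeping coat physics on; the side-coat transform is unconnected.
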AnimBlueprint /BRCosmetics/Characters/Player/Male/Medium/Bodies/M_MED_Polygon/Meshes/M_MED_Polygon_AnimBP
target_skeleton=&/M_MED_Polygon_Skeleton
graphs [
 AnimGraph {nodes [
   n1 = Get legSwingDelta_r() -> (legSwingDelta_r:double)
   "Transform (Modify) Bone - Bone: dyn_coat_side_r_base"(Alpha=@n1, Translation="0,0,0", Rotation="-10,-10,0", Scale="1,1,1") {
    rotationMode=BMM_Additive
    rotationSpace=BCS_BoneSpace
    translationMode=BMM_Additive
    translationSpace=BCS_BoneSpace
    alphaBoolBlend {blendOption=Linear}
    boneToModify {boneName=dyn_coat_side_r_base}
    rotation {pitch=-10, yaw=-10}
   }
   n2 = RigidBody(
    ComponentPose=Local To Component(LocalPose=Copy Pose From Mesh() {bUseAttachedParent=true})
    ExternalForce=Get Jump_Velocity() -> (Jump_Velocity:Vector)
   ) {
    bClampLinearTranslationLimitToRefPose=true
    bOverrideWorldGravity=true
    overridePhysicsAsset=&/M_MED_Polygon_Physics
    simulationSpace=BaseBoneSpace
    alphaBoolBlend {blendOption=Linear}
    baseBoneRef {boneName=pelvis}
    componentAppliedLinearAccClamp=<6000 6000 6000>
    componentLinearAccScale=<0.2 0.2 0.2>
    componentLinearVelScale=<0.2 0.2 0.2>
    overrideWorldGravity {z=-980}
   }
   n3 = RigidBody(ComponentPose=@n2, Alpha=Get Crouch_Alpha() -> (Crouch_Alpha:double)) {
    bClampLinearTranslationLimitToRefPose=true
    bOverrideWorldGravity=true
    overridePhysicsAsset=&/M_MED_Polygon_Coat_Physics
    simulationSpace=BaseBoneSpace
    alphaBoolBlend {blendOption=Linear}
    alphaScaleBiasClamp {bInterpResult=true, interpSpeedDecreasing=3, interpSpeedIncreasing=5}
    baseBoneRef {boneName=root}
    componentAppliedLinearAccClamp=<6000 6000 6000>
    componentLinearAccScale=<0.15 0.15 0.15>
    componentLinearVelScale=<0.15 0.15 0.15>
    overrideWorldGravity {z=-980}
   }
   n4 = "Constraint - dyn_coat_front_l_03 (PPPPPP)"!(
    ComponentPose=@n3
    Alpha=1
    ConstraintWeights_0=0.0028
    ConstraintWeights_1:float=0.0165
    ConstraintWeights_2:float=0.0259
    ConstraintWeights_3:float=0.1009
    ConstraintWeights_4:float=0.0104
    ConstraintWeights_5:float=0.8436
   ) -> (Pose) {
    alphaBoolBlend {blendOption=Linear}
    boneToModify {boneName=dyn_coat_front_l_03}
    constraintSetup [
     {targetBone {boneName=dyn_coat_front_l_parent}}
     {targetBone {boneName=dyn_coat_front_l_04}}
     {targetBone {boneName=dyn_coat_front_l_01}}
     {targetBone {boneName=dyn_coat_front_l_02}}
     {targetBone {boneName=dyn_coat_front_l_05}}
     {targetBone {boneName=thigh_l}}
    ]
    constraintWeights [1,1,1,1,1,1]
   }
   n5 = "Transform (Modify) Bone - Bone: dyn_coat_front_l_03"(ComponentPose=@n4, Alpha=Get LegRotXCrouch() -> (LegRotXCrouch:double), Translation="-0.000,5,-1", Rotation="-20,-6,0", Scale="1,1,1") {
    rotationMode=BMM_Additive
    rotationSpace=BCS_ParentBoneSpace
    translationMode=BMM_Additive
    translationSpace=BCS_BoneSpace
    alphaBoolBlend {blendOption=Linear}
    alphaScaleBiasClamp {bInterpResult=true, interpSpeedDecreasing=4, interpSpeedIncreasing=6}
    boneToModify {boneName=dyn_coat_front_l_03}
    rotation {pitch=-20, yaw=-6}
    translation {y=5, z=-1}
   }
   n6 = "Transform (Modify) Bone - Bone: dyn_coat_front_r_base"(ComponentPose=@n5, Alpha=@n1, Translation="0,0,0", Rotation="-50,-30,0", Scale="1,1,1") {
    rotationMode=BMM_Additive
    rotationSpace=BCS_BoneSpace
    translationMode=BMM_Additive
    translationSpace=BCS_BoneSpace
    alphaBoolBlend {blendOption=Linear}
    alphaScaleBiasClamp {bInterpResult=true, interpSpeedDecreasing=4, interpSpeedIncreasing=6}
    boneToModify {boneName=dyn_coat_front_r_base}
    rotation {pitch=-50, yaw=-30}
   }
   n7 = "Transform (Modify) Bone - Bone: dyn_pouch_l"(ComponentPose=@n6, Alpha=Get legSwingDelta_l() -> (legSwingDelta_l:double), Translation="-10,0,0", Rotation="20,0,20", Scale="1,1,1") {
    rotationMode=BMM_Additive
    rotationSpace=BCS_BoneSpace
    translationMode=BMM_Additive
    translationSpace=BCS_BoneSpace
    alphaBoolBlend {blendOption=Linear}
    alphaScaleBiasClamp {bInterpResult=true, interpSpeedIncreasing=50}
    boneToModify {boneName=dyn_pouch_l}
    rotation {pitch=20, roll=20}
    translation {x=-10}
   }
   n8 = "Transform (Modify) Bone - Bone: dyn_pouch_mid"(ComponentPose=@n7, Alpha=1, Translation="1,0,0", Rotation="0,0,0", Scale="1,1,1") {
    translationMode=BMM_Additive
    translationSpace=BCS_BoneSpace
    alphaBoolBlend {blendOption=Linear}
    alphaScaleBiasClamp {bInterpResult=true, interpSpeedDecreasing=4, interpSpeedIncreasing=6}
    boneToModify {boneName=dyn_pouch_mid}
    translation {x=1}
   }
   n9 = "Transform (Modify) Bone - Bone: dyn_pouch_r"(ComponentPose=@n8, Alpha=Get legSwingDelta_r(), Translation="-8,0,-2", Rotation="10,0,-20", Scale="1,1,1") {
    translationMode=BMM_Additive
    translationSpace=BCS_BoneSpace
    alphaBoolBlend {blendOption=Linear}
    alphaScaleBiasClamp {bInterpResult=true, interpSpeedDecreasing=50, interpSpeedIncreasing=8}
    boneToModify {boneName=dyn_pouch_r}
    rotation {pitch=10, roll=-20}
    translation {x=-8, z=-2}
   }
   n10 = Control Rig(Source=Component To Local(ComponentPose=@n9)) {
    alphaBoolBlend {blendOption=Linear}
    controlRigAssetReference {blueprintRigClass=&/M_MED_Polygon_Deform_CtrlRig_twists.M_MED_Polygon_Deform_CtrlRig_twists_C}
   }
   n11 = Control Rig!(
    Source=@n10
    Alpha
    KneeRot_fr_l_114deg:Vector
    KneeTrans_fr_l_114deg:Vector="0.258,-3.591,0.009"
    KneeScale_fr_l_114deg:Vector="1,1,1"
    KneeScale_fr_l_130deg:Vector="1,1,1"
    KneeRot_fr_l_130deg:Vector="0,0,-51.2"
    KneeTrans_fr_l_130deg:Vector="0.258,-3.591,0.009"
   ) -> (Pose) {
    alphaBoolBlend {blendOption=Linear}
    controlRigAssetReference {blueprintRigClass=&/M_MED_Polygon_Deform_CtrlRig_sculpters.M_MED_Polygon_Deform_CtrlRig_sculpters_C}
   }
   n12 = "Constraint - dyn_chest_strap_base (PPPPP)"!(
    ComponentPose=Local To Component(LocalPose=@n11)
    Alpha=1
    ConstraintWeights_0=0.0274
    ConstraintWeights_1:float=0.2164
    ConstraintWeights_2:float=0.2713
    ConstraintWeights_3:float=0.3599
    ConstraintWeights_4:float=0.1249
   ) -> (Pose) {
    alphaBoolBlend {blendOption=Linear}
    boneToModify {boneName=dyn_chest_strap_base}
    constraintSetup [
     {targetBone {boneName=spine_01}}
     {targetBone {boneName=spine_02}}
     {targetBone {boneName=spine_03}}
     {targetBone {boneName=spine_04}}
     {targetBone {boneName=spine_05}}
    ]
    constraintWeights [1,1,1,1,1]
   }
   Output Pose(Result=Component To Local(ComponentPose=@n12)) {layerGroup=None}
   Get LegRotXCrouch()
   Get legSwingDelta_l()
  ]}
 EventGraph {nodes [
   Event BlueprintUpdateAnimation()
   > n1 = Sequence()
   > n2 = Set Crouch_Alpha(Crouch_Alpha:double=SelectFloat(A=1, B=1, bPickA=Get bIsCrouching())) -> (Output_Get:double)
   n3 = GetOwnerSkeletalMeshComponent()
   n4 = Set legSwingDelta_l(
    execute=@n5.then
    legSwingDelta_l:double=MapRangeClamped(Value=GetTwistAndSwingAngleOfDeltaRotationFromRefPose(self=@n3, BoneName=thigh_l).OutSwingAngle, InRangeA=16, InRangeB=105)
   ) -> (Output_Get:double)
   Set LegRotXCrouch(execute=@n2.then, LegRotXCrouch:double=float * float(A=@n2, B=@n4)) -> (Output_Get:double)
   n5 = Set legSwingDelta_r(
    execute=@n1.then_1
    legSwingDelta_r:double=MapRangeClamped(Value=GetTwistAndSwingAngleOfDeltaRotationFromRefPose(self=@n3, BoneName=thigh_r).OutSwingAngle, InRangeA=177, InRangeB=73)
   ) -> (Output_Get:double)
   Event OnBeginJumping()
   > Set Jump_Velocity(Jump_Velocity:Vector="1500,1500,1500") -> (Output_Get:Vector)
   Event OnEndJumping()
   > Set Jump_Velocity(Jump_Velocity="0, 0, 0")
   Event OnEndFalling()
   > Set Jump_Velocity(Jump_Velocity="0, 0, 0")
   Event OnBeginFalling()
   > Set Jump_Velocity(Jump_Velocity="1500,1500,1500")
  ]}
]
variables [
 Crouch_Alpha {type=number}
 Jump_Velocity {type=Vector}
 LegRotXCrouch {type=number}
 legSwingDelta_l {type=number}
 legSwingDelta_r {type=number}
 Skel_Mesh {type=SkeletalMeshComponent}
]
