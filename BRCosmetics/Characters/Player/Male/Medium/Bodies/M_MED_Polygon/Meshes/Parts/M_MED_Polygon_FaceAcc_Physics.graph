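# Physics asset for the face accessory part of the M_MED_Polygon player character, providing physics bodies for its left and right dynamic hair strands.
# Each side is an independent three-bone chain, with a small sphere on the base bone and capsules on the parent and tip bones.
# Constraints link each bone to the one before it, sharing moderate, asymmetric swing limits and a very tight twist limit so strands bend with little roll.
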PhysicsAsset /BRCosmetics/Characters/Player/Male/Medium/Bodies/M_MED_Polygon/Meshes/Parts/M_MED_Polygon_FaceAcc_Physics
bodies {
 dyn_hair_strand_01_l [{shapeType=Capsule, length=2.5049, radius=1.4293, shapeName=dyn_hair_strand_01_l_capsule, center=<1.9606 0 0>, rotation=<rot 90 0 0>}]
 dyn_hair_strand_01_r [{shapeType=Capsule, length=2.5095, radius=1.2651, shapeName=dyn_hair_strand_01_r_capsule, center=<1.5323 0 0>, rotation=<rot 90 0 0>}]
 dyn_hair_strand_base_l [{shapeType=Sphere, radius=0.5, shapeName=dyn_hair_strand_base_l_sphere}]
 dyn_hair_strand_base_r [{shapeType=Sphere, radius=0.5, shapeName=dyn_hair_strand_base_r_sphere}]
 dyn_hair_strand_parent_l [{shapeType=Capsule, length=3.5464, radius=1.6267, shapeName=dyn_hair_strand_parent_l_capsule, center=<2.7852 0 0.0002>, rotation=<rot 90 0 0>}]
 dyn_hair_strand_parent_r [{
   shapeType=Capsule
   length=5.3022
   radius=1.8898
   shapeName=dyn_hair_strand_parent_r_capsule
   center=<4.5403 0 -0.0001>
   rotation=<rot 90 0 0>
  }]
}
constraints [
 {bone1Name=dyn_hair_strand_parent_l, bone2Name=dyn_hair_strand_base_l, swing1LimitDegrees=30, swing2LimitDegrees=20, twistLimitDegrees=3}
 {bone1Name=dyn_hair_strand_01_l, bone2Name=dyn_hair_strand_parent_l, swing1LimitDegrees=30, swing2LimitDegrees=20, twistLimitDegrees=3}
 {bone1Name=dyn_hair_strand_01_r, bone2Name=dyn_hair_strand_parent_r, swing1LimitDegrees=30, swing2LimitDegrees=20, twistLimitDegrees=3}
 {bone1Name=dyn_hair_strand_parent_r, bone2Name=dyn_hair_strand_base_r, swing1LimitDegrees=30, swing2LimitDegrees=20, twistLimitDegrees=3}
]
